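# EQS query for the MaskMaker boss that picks a random location for Spire Shards in a donut around the querier.
# A spiral ring generator takes inner and outer radii from query parameters and projects points using a water-avoiding nav filter.
# Tests enforce a scalable minimum 2D distance from saved shard locations, check water-avoiding paths and test world-static overlap, then a random test scores the points.
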EnvQuery /MaskMakerBoss/EQS/EQS_MaskMaker_SpireShardsDonutLocations_Random_Param
options [{
  generator {
   name=EnvQueryGenerator_Donut_0
   class=EnvQueryGenerator_Donut
   properties {
    bUseSpiralPattern=true
    center=&EnvQueryContext_Querier
    numberOfRings=5
    pointsPerRing=12
    innerRadius {
     dataBinding=&self:EnvQueryGenerator_Donut_0.AIDataProvider_QueryParams_0
     dataField=FloatValue
     defaultValue=300
    }
    outerRadius {
     dataBinding=&self:EnvQueryGenerator_Donut_0.AIDataProvider_QueryParams_1
     dataField=FloatValue
     defaultValue=400
    }
    projectionData {
     extentX=0
     navigationFilter=&NavFilter_Phoebe_AvoidWater_C
     postProjectionVerticalOffset=0
     projectDown=100
     projectUp=100
     serializedChannel=ECC_GameTraceChannel11
     traceChannel=TraceTypeQuery10
    }
   }
  }
  tests [
   EnvQueryTest_Distance_0 {
    class=EnvQueryTest_Distance
    properties {
     distanceTo=&/FortQueryContext_SpireShardsSavedLocations.FortQueryContext_SpireShardsSavedLocations_C
     filterType=Minimum
     scoringEquation=Linear
     scoringFactor=1
     testComment=""
     testMode=Distance2D
     testPurpose=Filter
     floatValueMin {
      dataBinding=&self:EnvQueryTest_Distance_0.FortAIDataProvider_ScalableFloat_0
      dataField=FloatValue
      defaultValue=200
     }
    }
   }
   EnvQueryTest_Pathfinding_0 {
    class=EnvQueryTest_Pathfinding
    properties {
     filterClass=&NavFilter_Phoebe_AvoidWater_C
     pathFromContext {dataBinding=null, dataField=BoolValue, defaultValue=true}
    }
   }
   EnvQueryTest_Overlap_0 {
    class=EnvQueryTest_Overlap
    properties {overlapData {extentX=50, extentY=50, extentZ=50, overlapChannel=ECC_WorldStatic, shapeOffset=<0 0 60>}}
   }
   FortQueryTest_Random_0 {class=FortQueryTest_Random, properties {floatValueMax=0, multipleContextScoreOp=Multiply, scoringFactor=1}}
  ]
 }]
