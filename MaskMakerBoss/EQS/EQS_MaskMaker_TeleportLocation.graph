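# EQS query that picks a teleport destination for the MaskMaker boss.
# It generates a grid inside a context-defined teleport area, projected onto navmesh with a no-water filter.
# Filters keep points within a maximum 2D distance of the player target, inside the AI leash and beyond a minimum distance from the portal center.
# Both distance limits bind to scalable-float providers, and a random test varies the pick.
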EnvQuery /MaskMakerBoss/EQS/EQS_MaskMaker_TeleportLocation
options [{
  generator {
   name=FortQueryGenerator_GridInBox_0
   class=FortQueryGenerator_GridInBox
   properties {
    bAutoSortTests=true
    bCanRunAsync=false
    boxExtentsContext=&/FortQueryContext_BP_MaskMaker_TeleportAreaExtent.FortQueryContext_BP_MaskMaker_TeleportAreaExtent_C
    boxLength=1000
    boxWidth=1000
    boxesContext=null
    envQueryResultNormalizationOption=Default
    generateAround=&/FortQueryContext_BP_MaskMaker_TeleportAreaCenter.FortQueryContext_BP_MaskMaker_TeleportAreaCenter_C
    minimumPointsOnLength=0
    minimumPointsOnWidth=0
    navDataOverrideContext=null
    optionName=""
    spaceBetween=300
    projectionData {
     bCanDisableTrace=true
     bCanProjectDown=true
     bCanTraceOnGeometry=true
     bCanTraceOnNavMesh=true
     bOnlyBlockingHits=true
     bTraceComplex=false
     extentX=0
     extentY=10
     extentZ=10
     navigationFilter=/Game/Athena/AI/Phoebe/NavFilter/NavFilter_Phoebe_NoWater.NavFilter_Phoebe_NoWater_C
     postProjectionVerticalOffset=0
     projectDown=1024
     projectUp=1024
     serializedChannel=ECC_Visibility
     traceChannel=TraceTypeQuery1
     traceMode=Navigation
     traceProfileName=None
     traceShape=Line
     versionNum=1
    }
   }
  }
  tests [
   EnvQueryTest_Distance_2 {
    class=EnvQueryTest_Distance
    properties {
     distanceTo=&/FortQueryContext_BP_MaskMakerPlayerTarget.FortQueryContext_BP_MaskMakerPlayerTarget_C
     filterType=Maximum
     floatValueMin=800
     scoringEquation=Linear
     scoringFactor=1
     testComment=""
     testMode=Distance2D
     testPurpose=Filter
     floatValueMax {
      dataBinding=&self:EnvQueryTest_Distance_2.FortAIDataProvider_ScalableFloat_0
      dataField=FloatValue
      defaultValue=275
     }
    }
   }
   FortQueryTest_InsideAIBotLeash_0 {class=FortQueryTest_InsideAIBotLeash, properties {testPurpose=Filter}}
   EnvQueryTest_Distance_0 {
    class=EnvQueryTest_Distance
    properties {
     distanceTo=&/FortQueryContext_BP_MaskMaker_PortalCenter.FortQueryContext_BP_MaskMaker_PortalCenter_C
     filterType=Minimum
     scoringEquation=Linear
     scoringFactor=1
     testComment=""
     testMode=Distance2D
     testPurpose=Filter
     floatValueMax {
      dataBinding=&self:EnvQueryTest_Distance_0.FortAIDataProvider_ScalableFloat_0
      dataField=FloatValue
      defaultValue=275
     }
     floatValueMin {
      dataBinding=&self:EnvQueryTest_Distance_0.FortAIDataProvider_ScalableFloat_1
      dataField=FloatValue
      defaultValue=750
     }
    }
   }
   FortQueryTest_Random_0 {class=FortQueryTest_Random, properties {floatValueMax=0, scoringFactor=1}}
  ]
 }]
