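; EQS query context for the MaskMaker boss that provides a single player target actor to environment queries.
; When the querier casts to FortPlayerPawnAthena, it returns that pawn's blackboard target actor if valid.
; When the cast fails, it falls back to a PlayerTargetEQSDummy actor found by class in the world.
; A valid pawn with no valid target does not use the dummy fallback.
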(fn ProvideSingleActor
  (params (QuerierObject :type Object%20Reference) (QuerierActor :type Actor%20Object%20Reference))
  (bind _asfort_player_pawn_athena (Utilities|Casting|CastToFortPlayerPawnAthena QuerierActor)
    (:then
      (bind _targetactor (Class|BPFLAICommonFunctions|GetTargetActorfromBlackboard _asfort_player_pawn_athena))
      (Utilities|IsValid _targetactor
        (:Is%20Valid
          (return _targetactor))))
    (:CastFailed
      (bind _returnvalue (Actor|GetActorOfClass "/MaskMakerBoss/EQS/CherrySmoke_MaskMaker_PlayerTargetEQSDummy.CherrySmoke_MaskMaker_PlayerTargetEQSDummy_C"))
      (Utilities|IsValid _returnvalue
        (:Is%20Valid
          (return _returnvalue))))))
