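; EQS query context for the MaskMaker boss that provides the portal center as a single location.
; If the querier casts to FortPlayerPawnAthena, it gets the portal target point actor through MaskMakerPawnInterface and returns that actor's location.
; If the cast fails, it falls back to GetActorOfClass for BP_PortalTargetPoint.
; An invalid target after a successful cast does not trigger that fallback.
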(fn ProvideSingleLocation
  (params (QuerierObject :type Object%20Reference) (QuerierActor :type Actor%20Object%20Reference))
  (bind _asfort_player_pawn_athena (Utilities|Casting|CastToFortPlayerPawnAthena QuerierActor)
    (:then
      (bind _actor (Class|MaskMakerPawnInterface|GetPortalTargetPoint(Message) _asfort_player_pawn_athena))
      (Utilities|IsValid _actor
        (:Is%20Valid
          (return (Transformation|GetActorLocation _actor)))))
    (:CastFailed
      (bind _returnvalue (Actor|GetActorOfClass "/MaskMakerBoss/EQS/BP_PortalTargetPoint.BP_PortalTargetPoint_C"))
      (Utilities|IsValid _returnvalue
        (:Is%20Valid
          (return (Transformation|GetActorLocation _returnvalue)))))))
