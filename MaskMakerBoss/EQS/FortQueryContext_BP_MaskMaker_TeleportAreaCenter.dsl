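; EQS query context for the MaskMaker boss that supplies the center of its teleport area as a single location.
; It casts the querier to a Fortnite player pawn, asks it for its teleport area box through MaskMakerPawnInterface, and returns that actor's location.
; Only when the cast fails does it use the first MaskMaker_TeleportArea_Box actor found in the world, with both paths checking validity first.
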(fn ProvideSingleLocation
  (params (QuerierObject :type Object%20Reference) (QuerierActor :type Actor%20Object%20Reference))
  (bind _asfort_player_pawn_athena (Utilities|Casting|CastToFortPlayerPawnAthena QuerierActor)
    (:then
      (bind _actorref (Class|MaskMakerPawnInterface|GetTeleportAreaBox(Message) _asfort_player_pawn_athena))
      (Utilities|IsValid _actorref
        (:Is%20Valid
          (return (Transformation|GetActorLocation _actorref)))))
    (:CastFailed
      (bind _returnvalue (Actor|GetActorOfClass "/MaskMakerBoss/EQS/MaskMaker_TeleportArea_Box.MaskMaker_TeleportArea_Box_C"))
      (Utilities|IsValid _returnvalue
        (:Is%20Valid
          (return (Transformation|GetActorLocation _returnvalue)))))))
