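; EQS query context that supplies grounded player locations around the MaskMaker boss's teleport area box.
; It sphere-traces at the box using its extent, ignoring saved targeted players, then saves new non-NonParticipant hits and flattens their positions to the portal target point's Z.
; When the querier isn't a FortPlayerPawnAthena, it uses EQS dummy actors and a world-found portal target point instead.
; An empty result returns the origin.
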(fn ProvideLocationsSet
  (params (QuerierObject :type Object%20Reference) (QuerierActor :type Actor%20Object%20Reference))
  (bind _playerlocationsgrounded (LocalVariables|GetPlayerLocationsGrounded))
  (bind _asfort_player_pawn_athena (Utilities|Casting|CastToFortPlayerPawnAthena QuerierActor)
    (:then
      (bind _output_get (LocalVariables|SetMaskMakerPawn _asfort_player_pawn_athena))
      (bind _actor (Class|MaskMakerPawnInterface|GetPortalTargetPoint(Message) _output_get))
      (bind (_returnvalue_x _returnvalue_y _returnvalue_z) (Math|Vector|BreakVector (Transformation|GetActorLocation _actor)))
      (LocalVariables|SetGroundedZLocation _returnvalue_z)
      (bind _actorref (Class|MaskMakerPawnInterface|GetTeleportAreaBox(Message) (LocalVariables|GetMaskMakerPawn)))
      (bind _platersref (Class|MaskMakerPawnInterface|GetSavedTargetedPlayers(Message) (LocalVariables|GetMaskMakerPawn)))
      (bind (_returnvalue_x_2 _returnvalue_y_2 _returnvalue_z_2) (Math|Vector|BreakVector (Components|Box|GetScaledBoxExtent (Actor|GetComponentbyClass _actorref "/Script/Engine.BoxComponent"))))
      (bind _outhits (Collision|MultiSphereTraceForObjects (Transformation|GetActorLocation _actorref) (+ (Transformation|GetActorLocation _actorref) "0.000000,0.000000,1.000000") (Math|Float|Max(Float) _returnvalue_x_2 _returnvalue_y_2 _returnvalue_z_2) (Utilities|Array|MakeArray "ObjectTypeQuery3") false _platersref))
      (for _array_element _outhits
        (bind _asplayer_pawn_athena (Utilities|Casting|CastToPlayerPawn_Athena (.HitActor (Collision|BreakHitResult _array_element))))
        (if (not (GameplayTags|HasAnyMatchingGameplayTags (Utilities|Casting|ToGameplayTagAssetInterface (.HitActor (Collision|BreakHitResult _array_element))) (GameplayTags|MakeGameplayTagContainerfromTag "(TagName=\"Pawn.Athena.NonParticipant\")")))
          (bind _platersref_2 (Class|MaskMakerPawnInterface|GetSavedTargetedPlayers(Message) (LocalVariables|GetMaskMakerPawn)))
          (Utilities|Array|AddUnique _platersref_2 _asplayer_pawn_athena)
          (bind (_returnvalue_x_3 _returnvalue_y_3) (Math|Vector|BreakVector (Transformation|GetActorLocation _asplayer_pawn_athena)))
          (Utilities|Array|AddUnique (LocalVariables|GetPlayerLocationsGrounded) (Math|Vector|MakeVector _returnvalue_x_3 _returnvalue_y_3 (LocalVariables|GetGroundedZLocation))))))
    (:CastFailed
      (bind _outactors (Actor|GetAllActorsOfClass "/MaskMakerBoss/EQS/BP_PortalTargetPoint.BP_PortalTargetPoint_C"))
      (Utilities|IsValid (Utilities|Array|Get(acopy) _outactors)
        (:Is%20Valid
          (bind (_returnvalue_x_4 _returnvalue_y_4 _returnvalue_z_3) (Math|Vector|BreakVector (Transformation|GetActorLocation (Utilities|Array|Get(acopy) _outactors))))
          (LocalVariables|SetGroundedZLocation _returnvalue_z_3)
          (bind _outactors_2 (Actor|GetAllActorsOfClass "/MaskMakerBoss/EQS/CherrySmoke_MaskMaker_PlayerTargetEQSDummy.CherrySmoke_MaskMaker_PlayerTargetEQSDummy_C"))
          (for _array_element_2 _outactors_2
            (bind (_returnvalue_x_5 _returnvalue_y_5) (Math|Vector|BreakVector (Transformation|GetActorLocation _array_element_2)))
            (Utilities|Array|AddUnique (LocalVariables|GetPlayerLocationsGrounded) (Math|Vector|MakeVector _returnvalue_x_5 _returnvalue_y_5 (LocalVariables|GetGroundedZLocation))))))))
  (if (> (Utilities|Array|Length _playerlocationsgrounded) 0)
    (return _playerlocationsgrounded)
    (else
      (return (Utilities|Array|MakeArray "0, 0, 0")))))
