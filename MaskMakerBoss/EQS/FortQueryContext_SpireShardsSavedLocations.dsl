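; EQS query context for the MaskMaker boss that supplies saved Spire Shard locations as a location set for environment queries.
; ProvideLocationsSet casts the querier actor to FortPlayerPawnAthena and sends it the MaskMakerPawnInterface GetSavedSpireShardsLocations message.
; The returned locations pass straight through as the context's result, with no filtering.
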(fn ProvideLocationsSet
  (params (QuerierObject :type Object%20Reference) (QuerierActor :type Actor%20Object%20Reference))
  (bind _asfort_player_pawn_athena (Utilities|Casting|CastToFortPlayerPawnAthena QuerierActor))
  (bind _locations (Class|MaskMakerPawnInterface|GetSavedSpireShardsLocations(Message) _asfort_player_pawn_athena))
  (return _locations))
